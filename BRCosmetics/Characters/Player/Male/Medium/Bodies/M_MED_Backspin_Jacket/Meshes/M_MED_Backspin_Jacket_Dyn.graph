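# Physics asset for the male medium Backspin jacket, defining bodies and joints for its dynamic coat bones.
# Seven capsule chains hang from pelvis-attached master bones, mostly with limited swing and tight twist, the centre back forking left and right.
# Cross-constraints tie neighbouring chains, mostly at lower segments, asymmetrically between sides.
# Pelvis, thigh and calf capsules model the body, and a huge root capsule sits below the origin like a floor.
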PhysicsAsset /BRCosmetics/Characters/Player/Male/Medium/Bodies/M_MED_Backspin_Jacket/Meshes/M_MED_Backspin_Jacket_Dyn
bodies {
 calf_l [{shapeType=Capsule, length=36.9375, radius=12.875, center=<-24.6324 -2.8287 0.0002>, rotation=<rot 88.0594 -8.7171 -14.8289>}]
 calf_r [{shapeType=Capsule, length=37.4375, radius=12.375, center=<24.6324 2.8287 -0.0002>, rotation=<rot 88.5756 -12.1951 -17.4256>}]
 dyn_coat_bk_01 [{shapeType=Capsule, length=20, radius=10.1, center=<6 0 10>, rotation=<rot 0 0 90>}]
 dyn_coat_bk_02 [{shapeType=Capsule, length=20, radius=10.1, center=<6 0 10>, rotation=<rot 0 0 90>}]
 dyn_coat_bk_03 [{shapeType=Capsule, length=20, radius=10.1, center=<6 0 10>, rotation=<rot 0 0 90>}]
 dyn_coat_bk_04 [{shapeType=Capsule, length=20, radius=10.1, center=<6 0 10>, rotation=<rot 0 0 90>}]
 dyn_coat_bk_MSTR [{shapeType=Capsule, length=0.2, radius=0.2}]
 dyn_coat_bk_l_01 [{shapeType=Capsule, length=23.25, radius=10.2, center=<6 0 10>, rotation=<rot 0 0 90>}]
 dyn_coat_bk_l_02 [{shapeType=Capsule, length=23.25, radius=10.2, center=<6 0 10>, rotation=<rot 0 0 90>}]
 dyn_coat_bk_l_03 [{shapeType=Capsule, length=23.25, radius=10.2, center=<6 0 10>, rotation=<rot 0 0 90>}]
 dyn_coat_bk_l_04 [{shapeType=Capsule, length=23.25, radius=10.2, center=<6 0 10>, rotation=<rot 0 0 90>}]
 dyn_coat_bk_l_05 [{shapeType=Capsule, length=27.75, radius=12.2, center=<6 0 10>, rotation=<rot 0 0 90>}]
 dyn_coat_bk_l_06 [{shapeType=Capsule, length=34.5, radius=12.95, center=<6 0 10>, rotation=<rot 0.137 -1.4259 85.6391>}]
 dyn_coat_bk_l_MSTR [{shapeType=Capsule, length=0.2, radius=0.2}]
 dyn_coat_bk_r_01 [{shapeType=Capsule, length=23.25, radius=10.2, center=<6 0 10>, rotation=<rot 0 0 90>}]
 dyn_coat_bk_r_02 [{shapeType=Capsule, length=23.25, radius=10.2, center=<6 0 10>, rotation=<rot 0 0 90>}]
 dyn_coat_bk_r_03 [{shapeType=Capsule, length=23.25, radius=10.2, center=<6 0 10>, rotation=<rot 0 0 90>}]
 dyn_coat_bk_r_04 [{shapeType=Capsule, length=23.25, radius=10.2, center=<6 0 10>, rotation=<rot 0 0 90>}]
 dyn_coat_bk_r_05 [{shapeType=Capsule, length=27.75, radius=12.2, center=<6 0 10>, rotation=<rot 0 0 90>}]
 dyn_coat_bk_r_06 [{shapeType=Capsule, length=25.75, radius=12.95, center=<6 0 10>, rotation=<rot 0.1199 1.1612 93.5531>}]
 dyn_coat_bk_r_MSTR [{shapeType=Capsule, length=0.2, radius=0.2}]
 dyn_coat_bkl_05 [{shapeType=Capsule, length=20, radius=10.1, center=<6 0 10>, rotation=<rot 0 0 90>}]
 dyn_coat_bkl_06 [{shapeType=Capsule, length=20, radius=10.1, center=<6 0 10>, rotation=<rot 0 0 90>}]
 dyn_coat_bkr_05 [{shapeType=Capsule, length=20, radius=10.1, center=<6 0 10>, rotation=<rot 0 0 90>}]
 dyn_coat_bkr_06 [{shapeType=Capsule, length=20, radius=10.1, center=<6 0 10>, rotation=<rot 0 0 90>}]
 dyn_coat_fr_l_01 [{shapeType=Capsule, length=7.28, radius=10, center=<6 0 8.5>, rotation=<rot 90 0 0>}]
 dyn_coat_fr_l_02 [{shapeType=Capsule, length=7.28, radius=10, center=<6 0 8.5>, rotation=<rot 90 0 0>}]
 dyn_coat_fr_l_03 [{shapeType=Capsule, length=7.28, radius=10, center=<6 0 8.5>, rotation=<rot 90 0 0>}]
 dyn_coat_fr_l_04 [{shapeType=Capsule, length=7.28, radius=10, center=<6 0 8.5>, rotation=<rot 90 0 0>}]
 dyn_coat_fr_l_05 [{shapeType=Capsule, length=7.28, radius=10, center=<6 0 8.5>, rotation=<rot 90 0 0>}]
 dyn_coat_fr_l_06 [{shapeType=Capsule, length=7.28, radius=10, center=<6 0 8.5>, rotation=<rot 90 0 0>}]
 dyn_coat_fr_l_MSTR [{shapeType=Capsule, length=0.2, radius=0.2}]
 dyn_coat_fr_r_01 [{shapeType=Capsule, length=7.25, radius=10, center=<6 0 8.5>, rotation=<rot 90 0 0>}]
 dyn_coat_fr_r_02 [{shapeType=Capsule, length=7.25, radius=10, center=<6 0 8.5>, rotation=<rot 90 0 0>}]
 dyn_coat_fr_r_03 [{shapeType=Capsule, length=7.25, radius=10, center=<6 0 8.5>, rotation=<rot 90 0 0>}]
 dyn_coat_fr_r_04 [{shapeType=Capsule, length=7.25, radius=10, center=<6 0 8.5>, rotation=<rot 90 0 0>}]
 dyn_coat_fr_r_05 [{shapeType=Capsule, length=7.25, radius=10, center=<6 0 8.5>, rotation=<rot 90 0 0>}]
 dyn_coat_fr_r_06 [{shapeType=Capsule, length=7.25, radius=10, center=<6 0 8.5>, rotation=<rot 90 0 0>}]
 dyn_coat_fr_r_MSTR [{shapeType=Capsule, length=0.2, radius=0.2}]
 dyn_coat_l_01 [{shapeType=Capsule, length=22.25, radius=10.25, center=<6 -2 5.5>, rotation=<rot 0 0 -110.3>}]
 dyn_coat_l_02 [{shapeType=Capsule, length=22.25, radius=10.25, center=<6 -2 5.5>, rotation=<rot 0 0 -110.3>}]
 dyn_coat_l_03 [{shapeType=Capsule, length=22.25, radius=10.25, center=<6 -2 5.5>, rotation=<rot 0 0 -110.3>}]
 dyn_coat_l_04 [{shapeType=Capsule, length=22.25, radius=10.25, center=<6 -2 5.5>, rotation=<rot 0 0 -110.3>}]
 dyn_coat_l_05 [{shapeType=Capsule, length=22.25, radius=10.25, center=<6 -2 5.5>, rotation=<rot 0 0 -110.3>}]
 dyn_coat_l_06 [{shapeType=Capsule, length=22.25, radius=10.25, center=<6 -2 5.5>, rotation=<rot 0 0 -110.3>}]
 dyn_coat_l_MSTR [{shapeType=Capsule, length=0.2, radius=0.2}]
 dyn_coat_r_01 [{shapeType=Capsule, length=22.25, radius=10.25, center=<6.0002 2 5.5>, rotation=<rot 0 0 110.3>}]
 dyn_coat_r_02 [{shapeType=Capsule, length=22.25, radius=10.25, center=<6.0002 2 5.5>, rotation=<rot 0 0 110.3>}]
 dyn_coat_r_03 [{shapeType=Capsule, length=22.25, radius=10.25, center=<6.0002 2 5.5>, rotation=<rot 0 0 110.3>}]
 dyn_coat_r_04 [{shapeType=Capsule, length=22.25, radius=10.25, center=<6.0002 2 5.5>, rotation=<rot 0 0 110.3>}]
 dyn_coat_r_05 [{shapeType=Capsule, length=22.25, radius=10.25, center=<6.0002 2 5.5>, rotation=<rot 0 0 110.3>}]
 dyn_coat_r_06 [{shapeType=Capsule, length=22.25, radius=10.25, center=<6.0002 2 5.5>, rotation=<rot 0 0 110.3>}]
 dyn_coat_r_MSTR [{shapeType=Capsule, length=0.2, radius=0.2}]
 pelvis [{shapeType=Capsule, length=9.5938, radius=10.4062, center=<-2.3213 0.0139 0>}]
 root [{shapeType=Capsule, length=2.75, radius=455, center=<-0.0002 -39.8738 -455.7143>, rotation=<rot 54.898 163.5935 -109.7889>}]
 thigh_l [{shapeType=Capsule, length=31.9688, radius=10.0938, center=<-19.2705 0.3803 0.3461>, rotation=<rot 89.7019 -178.6429 -179.9991>}]
 thigh_r [{shapeType=Capsule, length=31.9688, radius=10.0938, center=<19.2153 0.3957 0>, rotation=<rot 90 9.7524 5.817>}]
}
constraints [
 {bone1Name=dyn_coat_l_01, bone2Name=dyn_coat_l_MSTR, swing1LimitDegrees=45, swing2LimitDegrees=45, twistLimitDegrees=5}
 {bone1Name=dyn_coat_l_02, bone2Name=dyn_coat_l_01, swing1LimitDegrees=45, swing2LimitDegrees=45, twistLimitDegrees=5}
 {bone1Name=dyn_coat_l_03, bone2Name=dyn_coat_l_02, swing1LimitDegrees=45, swing2LimitDegrees=45, twistLimitDegrees=5}
 {bone1Name=dyn_coat_l_04, bone2Name=dyn_coat_l_03, swing1LimitDegrees=45, swing2LimitDegrees=45, twistLimitDegrees=5}
 {bone1Name=dyn_coat_l_05, bone2Name=dyn_coat_l_04, swing1LimitDegrees=45, swing2LimitDegrees=45, twistLimitDegrees=5}
 {bone1Name=dyn_coat_l_06, bone2Name=dyn_coat_l_05, swing1LimitDegrees=45, swing2LimitDegrees=45, twistLimitDegrees=5}
 {bone1Name=dyn_coat_fr_l_01, bone2Name=dyn_coat_fr_l_MSTR, swing1LimitDegrees=45, swing2LimitDegrees=45, twistLimitDegrees=5}
 {bone1Name=dyn_coat_fr_l_02, bone2Name=dyn_coat_fr_l_01, swing1LimitDegrees=45, swing2LimitDegrees=45, twistLimitDegrees=5}
 {bone1Name=dyn_coat_fr_l_03, bone2Name=dyn_coat_fr_l_02, swing1LimitDegrees=45, swing2LimitDegrees=45, twistLimitDegrees=5}
 {bone1Name=dyn_coat_fr_l_04, bone2Name=dyn_coat_fr_l_03, swing1LimitDegrees=45, swing2LimitDegrees=45, twistLimitDegrees=5}
 {bone1Name=dyn_coat_fr_l_05, bone2Name=dyn_coat_fr_l_04, swing1LimitDegrees=45, swing2LimitDegrees=45, twistLimitDegrees=5}
 {bone1Name=dyn_coat_fr_l_06, bone2Name=dyn_coat_fr_l_05, swing1LimitDegrees=45, swing2LimitDegrees=45, twistLimitDegrees=5}
 {bone1Name=dyn_coat_bk_01, bone2Name=dyn_coat_bk_MSTR, swing1LimitDegrees=45, swing2LimitDegrees=45, twistLimitDegrees=5}
 {bone1Name=dyn_coat_bk_02, bone2Name=dyn_coat_bk_01, swing1LimitDegrees=45, swing2LimitDegrees=45, twistLimitDegrees=5}
 {bone1Name=dyn_coat_bk_03, bone2Name=dyn_coat_bk_02, swing1LimitDegrees=45, swing2LimitDegrees=45, twistLimitDegrees=5}
 {bone1Name=dyn_coat_bk_04, bone2Name=dyn_coat_bk_03, swing1LimitDegrees=45, swing2LimitDegrees=45, twistLimitDegrees=5}
 {bone1Name=dyn_coat_bkl_05, bone2Name=dyn_coat_bk_04, swing1LimitDegrees=45, swing2LimitDegrees=45, twistLimitDegrees=5}
 {bone1Name=dyn_coat_bkl_06, bone2Name=dyn_coat_bkl_05, swing1LimitDegrees=45, swing2LimitDegrees=45, twistLimitDegrees=5}
 {bone1Name=dyn_coat_bkr_05, bone2Name=dyn_coat_bk_04, swing1LimitDegrees=45, swing2LimitDegrees=45, twistLimitDegrees=5}
 {bone1Name=dyn_coat_bkr_06, bone2Name=dyn_coat_bkr_05, swing1LimitDegrees=45, swing2LimitDegrees=45, twistLimitDegrees=5}
 {bone1Name=dyn_coat_bk_l_01, bone2Name=dyn_coat_bk_l_MSTR, swing1LimitDegrees=45, swing2LimitDegrees=45, twistLimitDegrees=5}
 {bone1Name=dyn_coat_bk_l_02, bone2Name=dyn_coat_bk_l_01, swing1LimitDegrees=45, swing2LimitDegrees=45, twistLimitDegrees=5}
 {bone1Name=dyn_coat_bk_l_03, bone2Name=dyn_coat_bk_l_02, swing1LimitDegrees=45, swing2LimitDegrees=45, twistLimitDegrees=5}
 {bone1Name=dyn_coat_bk_l_04, bone2Name=dyn_coat_bk_l_03, swing1LimitDegrees=45, swing2LimitDegrees=45, twistLimitDegrees=5}
 {bone1Name=dyn_coat_bk_l_05, bone2Name=dyn_coat_bk_l_04, swing1LimitDegrees=45, swing2LimitDegrees=45, twistLimitDegrees=5}
 {bone1Name=dyn_coat_bk_l_06, bone2Name=dyn_coat_bk_l_05, swing1LimitDegrees=45, swing2LimitDegrees=45, twistLimitDegrees=5}
 {bone1Name=dyn_coat_bk_r_01, bone2Name=dyn_coat_bk_r_MSTR, swing1LimitDegrees=45, swing2LimitDegrees=45, twistLimitDegrees=5}
 {bone1Name=dyn_coat_bk_r_02, bone2Name=dyn_coat_bk_r_01, swing1LimitDegrees=45, swing2LimitDegrees=45, twistLimitDegrees=5}
 {bone1Name=dyn_coat_bk_r_03, bone2Name=dyn_coat_bk_r_02, swing1LimitDegrees=45, swing2LimitDegrees=45, twistLimitDegrees=5}
 {bone1Name=dyn_coat_bk_r_04, bone2Name=dyn_coat_bk_r_03, swing1LimitDegrees=45, swing2LimitDegrees=45, twistLimitDegrees=5}
 {bone1Name=dyn_coat_bk_r_05, bone2Name=dyn_coat_bk_r_04, swing1LimitDegrees=45, swing2LimitDegrees=45, twistLimitDegrees=5}
 {bone1Name=dyn_coat_bk_r_06, bone2Name=dyn_coat_bk_r_05, swing1LimitDegrees=45, swing2LimitDegrees=45, twistLimitDegrees=5}
 {bone1Name=dyn_coat_fr_r_01, bone2Name=dyn_coat_fr_r_MSTR, swing1LimitDegrees=45, swing2LimitDegrees=45, twistLimitDegrees=5}
 {bone1Name=dyn_coat_fr_r_02, bone2Name=dyn_coat_fr_r_01, swing1LimitDegrees=45, swing2LimitDegrees=45, twistLimitDegrees=5}
 {bone1Name=dyn_coat_fr_r_03, bone2Name=dyn_coat_fr_r_02, swing1LimitDegrees=45, swing2LimitDegrees=45, twistLimitDegrees=5}
 {bone1Name=dyn_coat_fr_r_04, bone2Name=dyn_coat_fr_r_03, swing1LimitDegrees=45, swing2LimitDegrees=45, twistLimitDegrees=5}
 {bone1Name=dyn_coat_fr_r_05, bone2Name=dyn_coat_fr_r_04, swing1LimitDegrees=45, swing2LimitDegrees=45, twistLimitDegrees=5}
 {bone1Name=dyn_coat_fr_r_06, bone2Name=dyn_coat_fr_r_05, swing1LimitDegrees=45, swing2LimitDegrees=45, twistLimitDegrees=5}
 {bone1Name=dyn_coat_r_01, bone2Name=dyn_coat_r_MSTR, swing1LimitDegrees=45, swing2LimitDegrees=45, twistLimitDegrees=5}
 {bone1Name=dyn_coat_r_02, bone2Name=dyn_coat_r_01, swing1LimitDegrees=45, swing2LimitDegrees=45, twistLimitDegrees=5}
 {bone1Name=dyn_coat_r_03, bone2Name=dyn_coat_r_02, swing1LimitDegrees=45, swing2LimitDegrees=45, twistLimitDegrees=5}
 {bone1Name=dyn_coat_r_04, bone2Name=dyn_coat_r_03, swing1LimitDegrees=45, swing2LimitDegrees=45, twistLimitDegrees=5}
 {bone1Name=dyn_coat_r_05, bone2Name=dyn_coat_r_04, swing1LimitDegrees=45, swing2LimitDegrees=45, twistLimitDegrees=5}
 {bone1Name=dyn_coat_r_06, bone2Name=dyn_coat_r_05, swing1LimitDegrees=45, swing2LimitDegrees=45, twistLimitDegrees=5}
 {bone1Name=dyn_coat_l_MSTR, bone2Name=pelvis, swing1LimitDegrees=45, swing2LimitDegrees=45, twistLimitDegrees=5}
 {bone1Name=dyn_coat_fr_l_MSTR, bone2Name=pelvis, swing1LimitDegrees=45, swing2LimitDegrees=45, twistLimitDegrees=5}
 {bone1Name=dyn_coat_bk_MSTR, bone2Name=pelvis, swing1LimitDegrees=45, swing2LimitDegrees=45, twistLimitDegrees=5}
 {
  bone1Name=dyn_coat_bk_l_MSTR
  bone2Name=pelvis
  swing1LimitDegrees=45
  swing1Motion=Free
  swing2LimitDegrees=45
  swing2Motion=Free
  twistLimitDegrees=45
  twistMotion=Free
 }
 {
  bone1Name=dyn_coat_bk_r_MSTR
  bone2Name=pelvis
  swing1LimitDegrees=45
  swing1Motion=Free
  swing2LimitDegrees=45
  swing2Motion=Free
  twistLimitDegrees=45
  twistMotion=Free
 }
 {
  bone1Name=dyn_coat_fr_r_MSTR
  bone2Name=pelvis
  swing1LimitDegrees=45
  swing1Motion=Free
  swing2LimitDegrees=45
  swing2Motion=Free
  twistLimitDegrees=45
  twistMotion=Free
 }
 {
  bone1Name=dyn_coat_r_MSTR
  bone2Name=pelvis
  swing1LimitDegrees=45
  swing1Motion=Free
  swing2LimitDegrees=45
  swing2Motion=Free
  twistLimitDegrees=45
  twistMotion=Free
 }
 {
  bone1Name=thigh_l
  bone2Name=pelvis
  swing1LimitDegrees=45
  swing1Motion=Free
  swing2LimitDegrees=45
  swing2Motion=Free
  twistLimitDegrees=45
  twistMotion=Free
 }
 {
  bone1Name=calf_l
  bone2Name=thigh_l
  swing1LimitDegrees=45
  swing1Motion=Free
  swing2LimitDegrees=45
  swing2Motion=Free
  twistLimitDegrees=45
  twistMotion=Free
 }
 {
  bone1Name=thigh_r
  bone2Name=pelvis
  swing1LimitDegrees=45
  swing1Motion=Free
  swing2LimitDegrees=45
  swing2Motion=Free
  twistLimitDegrees=45
  twistMotion=Free
 }
 {
  bone1Name=calf_r
  bone2Name=thigh_r
  swing1LimitDegrees=45
  swing1Motion=Free
  swing2LimitDegrees=45
  swing2Motion=Free
  twistLimitDegrees=45
  twistMotion=Free
 }
 {
  bone1Name=pelvis
  bone2Name=root
  swing1LimitDegrees=45
  swing1Motion=Free
  swing2LimitDegrees=45
  swing2Motion=Free
  twistLimitDegrees=45
  twistMotion=Free
 }
 {
  bone1Name=dyn_coat_fr_l_04
  bone2Name=dyn_coat_l_04
  swing1LimitDegrees=45
  swing1Motion=Free
  swing2LimitDegrees=45
  swing2Motion=Free
  twistLimitDegrees=45
  twistMotion=Free
 }
 {
  bone1Name=dyn_coat_fr_l_05
  bone2Name=dyn_coat_l_05
  swing1LimitDegrees=45
  swing1Motion=Free
  swing2LimitDegrees=45
  swing2Motion=Free
  twistLimitDegrees=45
  twistMotion=Free
 }
 {
  bone1Name=dyn_coat_r_04
  bone2Name=dyn_coat_fr_r_04
  swing1LimitDegrees=45
  swing1Motion=Free
  swing2LimitDegrees=45
  swing2Motion=Free
  twistLimitDegrees=45
  twistMotion=Free
 }
 {
  bone1Name=dyn_coat_r_05
  bone2Name=dyn_coat_fr_r_05
  swing1LimitDegrees=45
  swing1Motion=Free
  swing2LimitDegrees=45
  swing2Motion=Free
  twistLimitDegrees=45
  twistMotion=Free
 }
 {bone1Name=dyn_coat_bkr_05, bone2Name=dyn_coat_bk_r_04, swing1LimitDegrees=45, swing2LimitDegrees=45, twistLimitDegrees=45}
 {bone1Name=dyn_coat_bkr_06, bone2Name=dyn_coat_bk_r_05, swing1LimitDegrees=45, swing2LimitDegrees=45, twistLimitDegrees=45}
 {
  bone1Name=dyn_coat_bkl_05
  bone2Name=dyn_coat_bk_l_04
  swing1LimitDegrees=45
  swing1Motion=Free
  swing2LimitDegrees=45
  swing2Motion=Free
  twistLimitDegrees=45
  twistMotion=Free
 }
 {
  bone1Name=dyn_coat_bkl_06
  bone2Name=dyn_coat_bk_l_05
  swing1LimitDegrees=45
  swing1Motion=Free
  swing2LimitDegrees=45
  swing2Motion=Free
  twistLimitDegrees=45
  twistMotion=Free
 }
 {
  bone1Name=dyn_coat_r_06
  bone2Name=dyn_coat_bk_r_06
  swing1LimitDegrees=45
  swing1Motion=Free
  swing2LimitDegrees=45
  swing2Motion=Free
  twistLimitDegrees=45
  twistMotion=Free
 }
 {
  bone1Name=dyn_coat_l_05
  bone2Name=dyn_coat_bk_l_05
  swing1LimitDegrees=45
  swing1Motion=Free
  swing2LimitDegrees=45
  swing2Motion=Free
  twistLimitDegrees=45
  twistMotion=Free
 }
 {bone1Name=dyn_coat_r_06, bone2Name=dyn_coat_bk_r_06, swing1LimitDegrees=45, swing2LimitDegrees=45, twistLimitDegrees=45}
 {bone1Name=dyn_coat_r_05, bone2Name=dyn_coat_bk_r_05, swing1LimitDegrees=45, swing2LimitDegrees=45, twistLimitDegrees=45}
 {
  bone1Name=dyn_coat_r_03
  bone2Name=dyn_coat_bk_r_03
  swing1LimitDegrees=45
  swing1Motion=Free
  swing2LimitDegrees=45
  swing2Motion=Free
  twistLimitDegrees=45
  twistMotion=Free
 }
 {bone1Name=dyn_coat_r_04, bone2Name=dyn_coat_bk_r_04, swing1LimitDegrees=45, swing2LimitDegrees=45, twistLimitDegrees=45}
 {bone1Name=dyn_coat_bkr_06, bone2Name=dyn_coat_bk_r_06, swing1LimitDegrees=45, swing2LimitDegrees=45, twistLimitDegrees=45}
]
