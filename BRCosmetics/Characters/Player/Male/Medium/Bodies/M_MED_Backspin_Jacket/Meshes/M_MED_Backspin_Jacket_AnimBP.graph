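# Post-process style Anim Blueprint for the Backspin Jacket player body, adding shoulder and coat secondary motion to the copied parent pose.
# Constraints and twist-driven additive offsets correct the shoulders, bypassed during front-end idle.
# Offset coat bones then feed a RigidBody simulation, reset shortly after skydiving begins or ends.
# Driving Ostrich, Ferret or Octopus vehicles disables simulation and shrinks coat bones.
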AnimBlueprint /BRCosmetics/Characters/Player/Male/Medium/Bodies/M_MED_Backspin_Jacket/Meshes/M_MED_Backspin_Jacket_AnimBP
target_skeleton=&/M_MED_Backspin_Jacket_Skeleton
graphs [
 AnimGraph {nodes [
   "Constraint - dyn_shldr_parent_r (PP)"!(ComponentPose, Alpha=1, ConstraintWeights_0=0.3, ConstraintWeights_1:float=0.7) -> (Pose) {
    alphaBoolBlend {blendOption=Linear}
    boneToModify {boneName=dyn_shldr_parent_r}
    constraintSetup [{targetBone {boneName=upperarm_twist_01_r}}, {targetBone {boneName=clavicle_r}}]
    constraintWeights [1,1]
   }
   n1 = "Transform (Modify) Bone - Bone: dyn_coat_bk_MSTR"(
    ComponentPose=Local To Component(LocalPose=Use cached pose 'main_SavedPose'())
    Alpha=1
    Translation="2.835,0.000,1.773"
    Rotation="5.594,0,0.000"
    Scale="1,1,1"
   ) {
    rotationMode=BMM_Additive
    rotationSpace=BCS_BoneSpace
    translationMode=BMM_Additive
    translationSpace=BCS_BoneSpace
    alphaBoolBlend {blendOption=Linear}
    boneToModify {boneName=dyn_coat_bk_MSTR}
    rotation {pitch=5.594}
    translation=<2.835 0 1.7732>
   }
   n2 = "Transform (Modify) Bone - Bone: dyn_coat_bk_r_MSTR"(ComponentPose=@n1, Alpha=1, Translation="-0.746,-1.508,0.717", Rotation="13.459,6.003,1.464", Scale="1,1,1") {
    rotationMode=BMM_Additive
    rotationSpace=BCS_BoneSpace
    translationMode=BMM_Additive
    translationSpace=BCS_BoneSpace
    alphaBoolBlend {blendOption=Linear}
    boneToModify {boneName=dyn_coat_bk_r_MSTR}
    rotation=<rot 13.459 6.0031 1.464>
    translation=<-0.746 -1.508 0.717>
   }
   n3 = "Transform (Modify) Bone - Bone: dyn_coat_bk_l_MSTR"(ComponentPose=@n2, Alpha=1, Translation="0.7,-1.5,0.5", Rotation="13.808,-4.073,-2.255", Scale="1,1,1") {
    rotationMode=BMM_Additive
    rotationSpace=BCS_BoneSpace
    translationMode=BMM_Additive
    translationSpace=BCS_BoneSpace
    alphaBoolBlend {blendOption=Linear}
    boneToModify {boneName=dyn_coat_bk_l_MSTR}
    rotation=<rot 13.808 -4.073 -2.255>
    translation=<0.7 -1.5 0.5>
   }
   n4 = RigidBody!(
    ComponentPose=@n3
    Alpha=Get in vehicle rigid alpha() -> (in vehicle rigid alpha:double)
    OverrideWorldGravity:Vector="0,0,-920"
    ExternalForce
   ) -> (Pose) {
    bClampLinearTranslationLimitToRefPose=true
    bOverrideWorldGravity=true
    overridePhysicsAsset=&/M_MED_Backspin_Jacket_Dyn
    simulationSpace=BaseBoneSpace
    alphaBoolBlend {blendOption=Linear}
    baseBoneRef {boneName=root}
    componentAppliedLinearAccClamp=<3000 3000 1000>
    componentLinearAccScale=<0.5 0.4 0.4>
    componentLinearVelScale=<0.5 0.4 0.4>
    overrideWorldGravity {z=-920}
   }
   Output Pose(Result=Component To Local(ComponentPose=@n4)) {layerGroup=None}
   n5 = "Constraint - dyn_shldr_parent_l (PPP)"!(
    ComponentPose=Local To Component(LocalPose=Use cached pose 'AnimCache'())
    Alpha=1
    ConstraintWeights_0=0.025
    ConstraintWeights_1:float=0.2925
    ConstraintWeights_2:float=0.6825
   ) -> (Pose) {
    alphaBoolBlend {blendOption=Linear}
    boneToModify {boneName=dyn_shldr_parent_l}
    constraintSetup [{targetBone {boneName=upperarm_twist_02_l}}, {targetBone {boneName=upperarm_l}}, {targetBone {boneName=upperarm_twist_01_l}}]
    constraintWeights [1,1,1]
   }
   n6 = "Constraint - dyn_shldr_parent_r (PPP)"!(ComponentPose=@n5, Alpha=1, ConstraintWeights_0=0.6825, ConstraintWeights_1:float=0.2925, ConstraintWeights_2:float=0.025) -> (Pose) {
    alphaBoolBlend {blendOption=Linear}
    boneToModify {boneName=dyn_shldr_parent_r}
    constraintSetup [{targetBone {boneName=upperarm_twist_01_r}}, {targetBone {boneName=upperarm_r}}, {targetBone {boneName=upperarm_twist_02_r}}]
    constraintWeights [1,1,1]
   }
   n7 = "Transform (Modify) Bone - Bone: dyn_shldr_01_l"(
    ComponentPose=@n6
    Alpha=Get shldr_l_alpha() -> (shldr_l_alpha:double)
    Translation="0.025,-0.5,-0.75"
    Rotation="-3.5,-1,-0.530"
    Scale="1,1,1"
   ) {
    rotationMode=BMM_Additive
    rotationSpace=BCS_BoneSpace
    scaleMode=BMM_Additive
    scaleSpace=BCS_BoneSpace
    translationMode=BMM_Additive
    translationSpace=BCS_BoneSpace
    alphaBoolBlend {blendOption=Linear}
    alphaScaleBiasClamp {bInterpResult=true, interpSpeedDecreasing=15, interpSpeedIncreasing=15}
    boneToModify {boneName=dyn_shldr_01_l}
    rotation=<rot -3.5 -1 -0.53>
    translation=<0.025 -0.5 -0.75>
   }
   n8 = "Transform (Modify) Bone - Bone: dyn_shldr_01_r"(ComponentPose=@n7, Alpha=Get shldr_r_alpha() -> (shldr_r_alpha:double), Translation="0.205,1.078,-0.261", Rotation="3,-2,0", Scale="1,1,1") {
    rotationMode=BMM_Additive
    rotationSpace=BCS_BoneSpace
    scaleMode=BMM_Additive
    scaleSpace=BCS_BoneSpace
    translationMode=BMM_Additive
    translationSpace=BCS_BoneSpace
    alphaBoolBlend {blendOption=Linear}
    alphaScaleBiasClamp {bInterpResult=true, interpSpeedDecreasing=25, interpSpeedIncreasing=25}
    boneToModify {boneName=dyn_shldr_01_r}
    rotation {pitch=3, yaw=-2}
    translation=<0.205 1.078 -0.261>
   }
   Save cached pose 'main_SavedPose'(Pose=Blend Poses by bool(
     BlendPose_0=Use cached pose 'AnimCache'()
     BlendPose_1=Component To Local(ComponentPose=@n8)
     bActiveValue=Get FrontEnd_Idle_On() -> (FrontEnd_Idle_On:bool)
    ) {blendType=Linear}) {cacheName=main_SavedPose}
   n9 = Get in vehicle scale alpha() -> (in vehicle scale alpha:double)
   n10 = "Transform (Modify) Bone - Bone: dyn_coat_l_02"(
    ComponentPose=Local To Component(LocalPose=Copy Pose From Mesh() {bUseAttachedParent=true})
    Alpha=@n9
    Translation="0,0,0"
    Rotation="0,0,0"
    Scale="0.1,0.1,0.1"
   ) {scaleMode=BMM_Replace, alphaBoolBlend {blendOption=Linear}, boneToModify {boneName=dyn_coat_l_02}, scale=<0.1 0.1 0.1>}
   n11 = "Transform (Modify) Bone - Bone: dyn_coat_fr_l_02"(ComponentPose=@n10, Alpha=@n9, Translation="0,0,0", Rotation="0,0,0", Scale="0.1,0.1,0.1") {scaleMode=BMM_Replace, alphaBoolBlend {blendOption=Linear}, boneToModify {boneName=dyn_coat_fr_l_02}, scale=<0.1 0.1 0.1>}
   n12 = "Transform (Modify) Bone - Bone: dyn_coat_bk_02"(ComponentPose=@n11, Alpha=@n9, Translation="0,0,0", Rotation="0,0,0", Scale="0.1,0.1,0.1") {scaleMode=BMM_Replace, alphaBoolBlend {blendOption=Linear}, boneToModify {boneName=dyn_coat_bk_02}, scale=<0.1 0.1 0.1>}
   n13 = "Transform (Modify) Bone - Bone: dyn_coat_bk_l_02"(ComponentPose=@n12, Alpha=@n9, Translation="0,0,0", Rotation="0,0,0", Scale="0.1,0.1,0.1") {scaleMode=BMM_Replace, alphaBoolBlend {blendOption=Linear}, boneToModify {boneName=dyn_coat_bk_l_02}, scale=<0.1 0.1 0.1>}
   n14 = "Transform (Modify) Bone - Bone: dyn_coat_bk_r_02"(ComponentPose=@n13, Alpha=@n9, Translation="0,0,0", Rotation="0,0,0", Scale="0.1,0.1,0.1") {scaleMode=BMM_Replace, alphaBoolBlend {blendOption=Linear}, boneToModify {boneName=dyn_coat_bk_r_02}, scale=<0.1 0.1 0.1>}
   n15 = "Transform (Modify) Bone - Bone: dyn_coat_fr_r_02"(ComponentPose=@n14, Alpha=@n9, Translation="0,0,0", Rotation="0,0,0", Scale="0.1,0.1,0.1") {scaleMode=BMM_Replace, alphaBoolBlend {blendOption=Linear}, boneToModify {boneName=dyn_coat_fr_r_02}, scale=<0.1 0.1 0.1>}
   n16 = "Transform (Modify) Bone - Bone: dyn_coat_r_02"(ComponentPose=@n15, Alpha=@n9, Translation="0,0,0", Rotation="0,0,0", Scale="0.1,0.1,0.1") {scaleMode=BMM_Replace, alphaBoolBlend {blendOption=Linear}, boneToModify {boneName=dyn_coat_r_02}, scale=<0.1 0.1 0.1>}
   Save cached pose 'AnimCache'(Pose=Component To Local(ComponentPose=@n16)) {cacheName=AnimCache}
  ]}
 EventGraph {nodes [
   n1 = Event BlueprintUpdateAnimation()
   > Get!(execute:Exec) -> (then:Exec, else:Exec, Skel_Mesh:Skeletal Mesh Component Object Reference)
   > n2 = Sequence!(execute) -> (then_0, then_1, then_2:Exec, then_3:Exec, then_4:Exec, then_5:Exec)
   > n3 = Branch(Condition="float > float"(A=Get PawnSpeed2D(), B=50) {is=title K2Node_CallFunction})
   > Set run_alpha(run_alpha:double=100) -> (Output_Get:double)
   Set run_alpha(execute=@n3.else, run_alpha=0)
   n4 = Branch(execute=@n2.then_1, Condition=Get bIsOstrichVehicleDriver() -> (bIsOstrichVehicleDriver:bool))
   Set in vehicle rigid alpha(execute=[@n4.then, @n5.then, @n6.then], in vehicle rigid alpha:double=0) -> (Output_Get:double)
   > Set in vehicle scale alpha(in vehicle scale alpha:double=1) -> (Output_Get:double)
   n5 = Branch(execute=@n4.else, Condition=Get bIsFerretVehicleDriver() -> (bIsFerretVehicleDriver:bool))
   n6 = Branch(execute=@n5.else, Condition=Get bIsOctopusVehicleDriver() -> (bIsOctopusVehicleDriver:bool))
   Set in vehicle rigid alpha(execute=@n6.else, in vehicle rigid alpha=1)
   > Set in vehicle scale alpha(in vehicle scale alpha=0)
   n7 = Get Skel_Mesh()
   Set shldr_r_alpha(
    execute=@n2.then_3
    shldr_r_alpha:double=GetAlphaFromReferenceTransformDelta(
     SkeletalMeshComponent=@n7
     BoneName=upperarm_r
     BaseName=clavicle_r
     DeltaComponentType=QuaternionTwist
     InputRange=(X=-70,Y=-120)
     TwistAxis="0,0,1"
    )
   ) -> (Output_Get:double)
   n8 = "Clamp (Float)"(
    Value=float * float(
     A=GetAlphaFromReferenceTransformDelta(
      SkeletalMeshComponent=@n7
      BoneName=upperarm_r
      BaseName=clavicle_r
      DeltaComponentType=QuaternionTwist
      InputRange=(X=0,Y=-90)
      TwistAxis="1,0,0"
     )
     B=GetAlphaFromReferenceTransformDelta(
      SkeletalMeshComponent=@n7
      BoneName=upperarm_r
      BaseName=clavicle_r
      DeltaComponentType=QuaternionTwist
      InputRange=(X=0,Y=80)
      TwistAxis="0,1,0"
     )
    )
    Min=0
    Max=1
   )
   > n9 = Branch(Condition="float <= float"!(A:double=@n8, B:double=0.1) -> (ReturnValue:bool) {is=title K2Node_CallFunction})
   > Set shldr_r_roll_alpha(shldr_r_roll_alpha:double=@n8) -> (Output_Get:double)
   Set shldr_r_roll_alpha(
    execute=@n9.else
    shldr_r_roll_alpha=FInterpTo(
     Current=Get shldr_r_roll_alpha() -> (shldr_r_roll_alpha:double)
     Target=@n8
     DeltaTime=Get delta_time() -> (delta_time:double)
     InterpSpeed=1.5
    )
   )
   n10 = Get Skel_Mesh()
   Set shldr_l_alpha(
    execute=@n2.then_2
    shldr_l_alpha:double=GetAlphaFromReferenceTransformDelta(
     SkeletalMeshComponent=@n10
     BoneName=upperarm_l
     BaseName=clavicle_l
     DeltaComponentType=QuaternionTwist
     InputRange=(X=-10,Y=-65)
     TwistAxis="0,0,1"
    )
   ) -> (Output_Get:double)
   n11 = "Clamp (Float)"(
    Value=float * float(
     A=GetAlphaFromReferenceTransformDelta(
      SkeletalMeshComponent=@n10
      BoneName=upperarm_l
      BaseName=clavicle_l
      DeltaComponentType=QuaternionTwist
      InputRange=(X=0,Y=-90)
      TwistAxis="1,0,0"
     )
     B=GetAlphaFromReferenceTransformDelta(
      SkeletalMeshComponent=@n10
      BoneName=upperarm_l
      BaseName=clavicle_l
      DeltaComponentType=QuaternionTwist
      InputRange=(X=0,Y=80)
      TwistAxis="0,1,0"
     )
    )
    Min=0
    Max=1
   )
   > n12 = Branch(Condition="float <= float"(A=@n11, B=0.1))
   > Set shldr_r_roll_alpha(shldr_r_roll_alpha="Clamp (Float)"(Value=float * float(A=0, B=0), Min=0, Max=1))
   Set shldr_r_roll_alpha(execute=@n12.else, shldr_r_roll_alpha=FInterpTo(Current=Get shldr_r_roll_alpha(), Target=@n11, DeltaTime=Get delta_time(), InterpSpeed=1.5))
   n13 = Get AnimInst() -> (AnimInst:Anim Instance Object Reference)
   Is Valid(exec=@n2.then_4, InputObject=@n13)
   > Set FrontEnd_Idle_On(FrontEnd_Idle_On:bool="float > float"(A=GetCurveValue(self=@n13, CurveName=FrontEnd_Idle_Bypass), B=0)) -> (Output_Get:bool)
   n14 = Event OnEndSkydiving()
   Delay(execute=[@n14.then, @n18.then], Duration=0.15)
   > ResetDynamics(InTeleportType=TeleportPhysics)
   > ResetDynamics(InTeleportType=ResetPhysics)
   Event BlueprintInitializeAnimation()
   n15 = GetOwnerSkeletalMeshComponent()
   > Is Valid(InputObject=@n15)
   > Set Skel_Mesh(Skel_Mesh=@n15)
   n16 = Event OnNotifyPartsStopped!() -> (OutputDelegate, then, Context:Name=None) {is=title K2Node_Event}
   > Branch(Condition="Equal (Name)"!(A:Name=@n16.Context, B:Name=Backspin_Idle) -> (ReturnValue:bool) {is=title K2Node_CallFunction})
   > Set FrontEnd_Idle_On(FrontEnd_Idle_On=false)
   n17 = Event OnNotifyPartsStarted!() -> (OutputDelegate, then, Context:Name=None) {is=title K2Node_Event}
   > Branch(Condition="Equal (Name)"(A=@n17.Context, B=Backspin_Idle))
   > Set FrontEnd_Idle_On(FrontEnd_Idle_On=true)
   n18 = Event OnBeginSkydiving()
   "float > float"(A=GetCurveValue(self=@n13, CurveName=Disable_Custom), B=0)
   Event OnParachuteOpened()
   Set delta_time(delta_time:double=@n1.DeltaTimeX) -> (Output_Get:double)
   Set GravityOverride_b(GravityOverride_b:Vector="0, 0, 0") -> (Output_Get:Vector)
  ]}
]
variables [
 AnimInst {type=AnimInstance, default=null}
 ClothAlpha_b {type=number}
 crouch_skirt_alpha {type=number}
 delta_time {type=number}
 emote_active {type=boolean}
 FrontEnd_Idle_On {type=boolean}
 Gravity Override Params {type=GravityOverrideParamsStruct}
 GravityOverride_b {type=Vector}
 in vehicle rigid alpha {type=number}
 in vehicle scale alpha {type=number}
 Pawn {type=FortPlayerPawn}
 run_alpha {type=number}
 shldr_l_alpha {type=number}
 shldr_r_alpha {type=number}
 shldr_r_pouch_alpha {type=number}
 shldr_r_roll_alpha {type=number}
 side_cloth_alpha {type=number}
 Skel_Mesh {type=SkeletalMeshComponent}
 Skeletal_Mesh {type=SkeletalMeshComponent}
 VectorForce_b {type=Vector}
 wearing_long_cape {type=boolean}
]
